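# Physics asset for the M_MED_Disguise backpack cosmetic, defining bodies and joint limits for its flap, lollipop and strap secondary motion.
# Each part uses a capsule body constrained to its own tiny MSTR root capsule.
# The lollipop also carries three child capsules, two loops and a flop, that swing more widely than their parent.
# The flap behaves like a hinge, with one wide swing axis, while the strap is tightly limited.
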PhysicsAsset /BRCosmetics/Accessories/FORT_Backpacks/Backpack_M_MED_Disguise/Meshes/M_MED_Disguise_Pack_Physics
bodies {
 dyn_flap [{shapeType=Capsule, length=12.1834, radius=7.6921, shapeName=dyn_flap_capsule, center=<5.9395 2.673 -0.9958>, rotation=<rot 7.492 -90 -4.365>}]
 dyn_flap_MSTR [{shapeType=Capsule, length=0.505, radius=0.505, shapeName=dyn_flap_MSTR_capsule, rotation=<rot 0 -90 0>}]
 dyn_lollipop [{
   shapeType=Capsule
   length=4.6692
   radius=4.1904
   shapeName=dyn_lollipop_capsule
   center=<6.0866 0.166 0.4051>
   rotation=<rot 0 90.6582 -86.7694>
  }]
 dyn_lollipopFlop [{shapeType=Capsule, length=2.255, radius=0.505, shapeName=dyn_lollipopFlop_capsule, center=<2.804 -0.0001 0>, rotation=<rot 90 0 0>}]
 dyn_lollipopLoop_B [{shapeType=Capsule, length=2.0246, radius=0.8097, shapeName=dyn_lollipopLoop_B_capsule, center=<1.7798 -0.0001 0>, rotation=<rot 90 0 0>}]
 dyn_lollipopLoop_F [{shapeType=Capsule, length=1.8244, radius=0.9444, shapeName=dyn_lollipopLoop_F_capsule, center=<1.7798 -0.0001 0>, rotation=<rot 90 0 0>}]
 dyn_lollipop_MSTR [{shapeType=Capsule, length=0.505, radius=0.505, shapeName=dyn_lollipop_MSTR_capsule, rotation=<rot 0 -90 0>}]
 dyn_strap [{shapeType=Capsule, length=3.0857, radius=2.5505, shapeName=dyn_strap_capsule, center=<2.3219 0 0>, rotation=<rot 90 0 0>}]
 dyn_strap_MSTR [{shapeType=Capsule, length=0.505, radius=0.505, shapeName=dyn_strap_MSTR_capsule, rotation=<rot 0 -90 0>}]
}
constraints [
 {bone1Name=dyn_lollipop, bone2Name=dyn_lollipop_MSTR, swing1LimitDegrees=15, swing2LimitDegrees=5, twistLimitDegrees=5}
 {bone1Name=dyn_lollipopLoop_F, bone2Name=dyn_lollipop, swing1LimitDegrees=25, swing2LimitDegrees=25, twistLimitDegrees=2}
 {bone1Name=dyn_lollipopLoop_B, bone2Name=dyn_lollipop, swing1LimitDegrees=25, swing2LimitDegrees=25, twistLimitDegrees=2}
 {bone1Name=dyn_lollipopFlop, bone2Name=dyn_lollipop, swing1LimitDegrees=25, swing2LimitDegrees=25, twistLimitDegrees=2}
 {bone1Name=dyn_flap, bone2Name=dyn_flap_MSTR, swing1LimitDegrees=45, swing2LimitDegrees=0, twistLimitDegrees=2}
 {bone1Name=dyn_strap, bone2Name=dyn_strap_MSTR, swing1LimitDegrees=5, swing2LimitDegrees=5, twistLimitDegrees=2}
]
