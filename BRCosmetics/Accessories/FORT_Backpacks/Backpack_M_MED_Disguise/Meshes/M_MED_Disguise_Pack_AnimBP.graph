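# Animation Blueprint for the M_MED_Disguise backpack cosmetic, adding secondary motion to its flap, candy and lollipop bones.
# A RigidBody simulation feeds bone offsets for flap yaw, falling, melee weapons and front-end posing, cached as 'Sim'.
# Two Anim Dynamics chains over the candy bones read that cache and blend on skydiving; the skydive chain mostly adds linear springs.
# EventGraph updates the alphas, including the skydive chain's front-end-dependent strength.
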AnimBlueprint /BRCosmetics/Accessories/FORT_Backpacks/Backpack_M_MED_Disguise/Meshes/M_MED_Disguise_Pack_AnimBP
target_skeleton=&/M_MED_Disguise_Pack_Skeleton
graphs [
 AnimGraph {nodes [
   n1 = Anim Dynamics - Bone: dyn_candy1!(ComponentPose=Local To Component(LocalPose=Use cached pose 'Sim'()), ExternalForce="0,0,0") -> (Pose) {
    alpha=0.5
    angularDampingOverride=0.998
    angularSpringConstant=100
    bAngularSpring=true
    bGravityOverrideInSimSpace=true
    bOverrideAngularDamping=true
    gravityScale=0.5
    simulationSpace=RootRelative
    alphaBoolBlend {blendOption=Linear}
    boundBone {boneName=dyn_candy1}
    componentAppliedLinearAccClamp=<3000 3000 3000>
    componentLinearAccScale=<0.3 0.3 0.4>
    componentLinearVelScale=<0.2 0.2 0.2>
    physicsBodyDefinitions [{
      sphereCollisionRadius=0
      boundBone {boneName=dyn_candy1}
      boxExtents=<8 30 30>
      constraintSetup {angularLimitsMax=<5 5 5>, angularLimitsMin=<-5 -5 -5>, angularTarget=<1 0 0>}
      localJointOffset=<0 0 -6>
     }]
   }
   n2 = Anim Dynamics - Bone: dyn_candy2!(ComponentPose=@n1, ExternalForce="0,0,0") -> (Pose) {
    alpha=0.35
    angularDampingOverride=0.998
    angularSpringConstant=100
    bAngularSpring=true
    bGravityOverrideInSimSpace=true
    bOverrideAngularDamping=true
    gravityScale=0.5
    simulationSpace=RootRelative
    alphaBoolBlend {blendOption=Linear}
    boundBone {boneName=dyn_candy2}
    componentAppliedLinearAccClamp=<3000 3000 3000>
    componentLinearAccScale=<0.4 0.4 0.5>
    componentLinearVelScale=<0.3 0.3 0.3>
    physicsBodyDefinitions [{
      sphereCollisionRadius=0
      boundBone {boneName=dyn_candy2}
      boxExtents=<12 8 30>
      constraintSetup {angularLimitsMax=<20 10 5>, angularLimitsMin=<-20 -10 -5>, angularTarget=<1 0 0>}
      localJointOffset=<0 0 -6>
     }]
   }
   n3 = Anim Dynamics - Bone: dyn_candy3!(ComponentPose=@n2, ExternalForce="0,0,0") -> (Pose) {
    alpha=0.35
    angularDampingOverride=0.998
    angularSpringConstant=100
    bAngularSpring=true
    bGravityOverrideInSimSpace=true
    bOverrideAngularDamping=true
    gravityScale=0.5
    simulationSpace=RootRelative
    alphaBoolBlend {blendOption=Linear}
    boundBone {boneName=dyn_candy3}
    componentAppliedLinearAccClamp=<3000 3000 3000>
    componentLinearAccScale=<0.4 0.4 0.5>
    componentLinearVelScale=<0.3 0.3 0.3>
    physicsBodyDefinitions [{
      sphereCollisionRadius=0
      boundBone {boneName=dyn_candy3}
      boxExtents=<12 8 25>
      constraintSetup {angularLimitsMax=<20 10 5>, angularLimitsMin=<-20 -10 -5>, angularTarget=<1 0 0>}
      localJointOffset=<0 0 -6>
     }]
   }
   n4 = Anim Dynamics - Bone: dyn_candy4!(ComponentPose=@n3, ExternalForce="0,0,0") -> (Pose) {
    alpha=0.35
    angularDampingOverride=0.998
    angularSpringConstant=100
    bAngularSpring=true
    bGravityOverrideInSimSpace=true
    bOverrideAngularDamping=true
    gravityScale=0.5
    simulationSpace=RootRelative
    alphaBoolBlend {blendOption=Linear}
    boundBone {boneName=dyn_candy4}
    componentAppliedLinearAccClamp=<3000 3000 3000>
    componentLinearAccScale=<0.4 0.4 0.5>
    componentLinearVelScale=<0.3 0.3 0.3>
    physicsBodyDefinitions [{
      sphereCollisionRadius=0
      boundBone {boneName=dyn_candy4}
      boxExtents=<12 8 25>
      constraintSetup {angularLimitsMax=<20 10 5>, angularLimitsMin=<-20 -10 -5>, angularTarget=<1 0 0>}
      localJointOffset=<0 0 -6>
     }]
   }
   n5 = Anim Dynamics - Bone: dyn_candy5!(ComponentPose=@n4, ExternalForce="0,0,0") -> (Pose) {
    alpha=0.35
    angularDampingOverride=0.998
    angularSpringConstant=100
    bAngularSpring=true
    bGravityOverrideInSimSpace=true
    bOverrideAngularDamping=true
    gravityScale=0.5
    simulationSpace=RootRelative
    alphaBoolBlend {blendOption=Linear}
    boundBone {boneName=dyn_candy5}
    componentAppliedLinearAccClamp=<3000 3000 3000>
    componentLinearAccScale=<0.4 0.4 0.5>
    componentLinearVelScale=<0.3 0.3 0.3>
    physicsBodyDefinitions [{
      sphereCollisionRadius=0
      boundBone {boneName=dyn_candy5}
      boxExtents=<12 8 35>
      constraintSetup {angularLimitsMax=<20 10 5>, angularLimitsMin=<-20 -10 -5>, angularTarget=<1 0 0>}
      localJointOffset=<0 0 -6>
     }]
   }
   n6 = Anim Dynamics - Bone: dyn_candy6!(ComponentPose=@n5, ExternalForce="0,0,0") -> (Pose) {
    alpha=0.35
    angularDampingOverride=0.998
    angularSpringConstant=100
    bAngularSpring=true
    bGravityOverrideInSimSpace=true
    bOverrideAngularDamping=true
    gravityScale=0.5
    simulationSpace=RootRelative
    alphaBoolBlend {blendOption=Linear}
    boundBone {boneName=dyn_candy6}
    componentAppliedLinearAccClamp=<3000 3000 3000>
    componentLinearAccScale=<0.4 0.4 0.5>
    componentLinearVelScale=<0.3 0.3 0.3>
    physicsBodyDefinitions [{
      sphereCollisionRadius=0
      boundBone {boneName=dyn_candy6}
      boxExtents=<12 8 25>
      constraintSetup {angularLimitsMax=<20 10 5>, angularLimitsMin=<-20 -10 -5>, angularTarget=<1 0 0>}
      localJointOffset=<0 0 -6>
     }]
   }
   n7 = Anim Dynamics - Bone: dyn_candy8!(ComponentPose=@n6, ExternalForce="0,0,0") -> (Pose) {
    alpha=0.35
    angularDampingOverride=0.998
    angularSpringConstant=100
    bAngularSpring=true
    bGravityOverrideInSimSpace=true
    bOverrideAngularDamping=true
    gravityScale=0.5
    simulationSpace=RootRelative
    alphaBoolBlend {blendOption=Linear}
    boundBone {boneName=dyn_candy8}
    componentAppliedLinearAccClamp=<3000 3000 3000>
    componentLinearAccScale=<0.4 0.4 0.5>
    componentLinearVelScale=<0.3 0.3 0.3>
    physicsBodyDefinitions [{
      sphereCollisionRadius=0
      boundBone {boneName=dyn_candy8}
      boxExtents=<12 8 30>
      constraintSetup {angularLimitsMax=<20 10 5>, angularLimitsMin=<-20 -10 -5>, angularTarget=<1 0 0>}
      localJointOffset=<0 0 -6>
     }]
   }
   n8 = Anim Dynamics - Bone: dyn_candy7!(ComponentPose=@n7, ExternalForce="0,0,0") -> (Pose) {
    alpha=0.35
    angularDampingOverride=0.998
    angularSpringConstant=100
    bAngularSpring=true
    bGravityOverrideInSimSpace=true
    bOverrideAngularDamping=true
    gravityScale=0.5
    simulationSpace=RootRelative
    alphaBoolBlend {blendOption=Linear}
    boundBone {boneName=dyn_candy7}
    componentAppliedLinearAccClamp=<3000 3000 3000>
    componentLinearAccScale=<0.4 0.4 0.5>
    componentLinearVelScale=<0.3 0.3 0.3>
    physicsBodyDefinitions [{
      sphereCollisionRadius=0
      boundBone {boneName=dyn_candy7}
      boxExtents=<12 8 35>
      constraintSetup {angularLimitsMax=<20 10 5>, angularLimitsMin=<-20 -10 -5>, angularTarget=<1 0 0>}
      localJointOffset=<0 0 -6>
     }]
   }
   n9 = Anim Dynamics - Bone: dyn_candy9!(ComponentPose=@n8, ExternalForce="0,0,0") -> (Pose) {
    alpha=0.35
    angularDampingOverride=0.998
    angularSpringConstant=100
    bAngularSpring=true
    bGravityOverrideInSimSpace=true
    bOverrideAngularDamping=true
    gravityScale=0.5
    simulationSpace=RootRelative
    alphaBoolBlend {blendOption=Linear}
    boundBone {boneName=dyn_candy9}
    componentAppliedLinearAccClamp=<3000 3000 3000>
    componentLinearAccScale=<0.4 0.4 0.5>
    componentLinearVelScale=<0.3 0.3 0.3>
    physicsBodyDefinitions [{
      sphereCollisionRadius=0
      boundBone {boneName=dyn_candy9}
      boxExtents=<12 8 25>
      constraintSetup {angularLimitsMax=<20 10 5>, angularLimitsMin=<-20 -10 -5>, angularTarget=<1 0 0>}
      localJointOffset=<0 0 -6>
     }]
   }
   n10 = Anim Dynamics - Bone: dyn_candy10!(ComponentPose=@n9, ExternalForce="0,0,0") -> (Pose) {
    alpha=0.35
    angularDampingOverride=0.998
    angularSpringConstant=100
    bAngularSpring=true
    bGravityOverrideInSimSpace=true
    bOverrideAngularDamping=true
    gravityScale=0.5
    simulationSpace=RootRelative
    alphaBoolBlend {blendOption=Linear}
    boundBone {boneName=dyn_candy10}
    componentAppliedLinearAccClamp=<3000 3000 3000>
    componentLinearAccScale=<0.4 0.4 0.5>
    componentLinearVelScale=<0.3 0.3 0.3>
    physicsBodyDefinitions [{
      sphereCollisionRadius=0
      boundBone {boneName=dyn_candy10}
      boxExtents=<12 8 20>
      constraintSetup {angularLimitsMax=<20 10 5>, angularLimitsMin=<-20 -10 -5>, angularTarget=<1 0 0>}
      localJointOffset=<0 0 -6>
     }]
   }
   n11 = Anim Dynamics - Bone: dyn_candy11!(ComponentPose=@n10, ExternalForce="0,0,0") -> (Pose) {
    alpha=0.35
    angularDampingOverride=0.998
    angularSpringConstant=100
    bAngularSpring=true
    bGravityOverrideInSimSpace=true
    bOverrideAngularDamping=true
    gravityScale=0.5
    simulationSpace=RootRelative
    alphaBoolBlend {blendOption=Linear}
    boundBone {boneName=dyn_candy11}
    componentAppliedLinearAccClamp=<3000 3000 3000>
    componentLinearAccScale=<0.4 0.4 0.5>
    componentLinearVelScale=<0.3 0.3 0.3>
    physicsBodyDefinitions [{
      sphereCollisionRadius=0
      boundBone {boneName=dyn_candy11}
      boxExtents=<12 8 30>
      constraintSetup {angularLimitsMax=<20 10 5>, angularLimitsMin=<-20 -10 -5>, angularTarget=<1 0 0>}
      localJointOffset=<0 0 -6>
     }]
   }
   n12 = Anim Dynamics - Bone: dyn_candy12!(ComponentPose=@n11, ExternalForce="0,0,0") -> (Pose) {
    alpha=0.35
    angularDampingOverride=0.998
    angularSpringConstant=100
    bAngularSpring=true
    bGravityOverrideInSimSpace=true
    bOverrideAngularDamping=true
    gravityScale=0.5
    simulationSpace=RootRelative
    alphaBoolBlend {blendOption=Linear}
    boundBone {boneName=dyn_candy12}
    componentAppliedLinearAccClamp=<3000 3000 3000>
    componentLinearAccScale=<0.4 0.4 0.5>
    componentLinearVelScale=<0.3 0.3 0.3>
    physicsBodyDefinitions [{
      sphereCollisionRadius=0
      boundBone {boneName=dyn_candy12}
      boxExtents=<12 8 30>
      constraintSetup {angularLimitsMax=<20 10 5>, angularLimitsMin=<-20 -10 -5>, angularTarget=<1 0 0>}
      localJointOffset=<0 0 -6>
     }]
   }
   n13 = Anim Dynamics - Bone: dyn_candy13!(ComponentPose=@n12, ExternalForce="0,0,0") -> (Pose) {
    alpha=0.35
    angularDampingOverride=0.998
    angularSpringConstant=100
    bAngularSpring=true
    bGravityOverrideInSimSpace=true
    bOverrideAngularDamping=true
    gravityScale=0.5
    simulationSpace=RootRelative
    alphaBoolBlend {blendOption=Linear}
    boundBone {boneName=dyn_candy13}
    componentAppliedLinearAccClamp=<3000 3000 3000>
    componentLinearAccScale=<0.4 0.4 0.5>
    componentLinearVelScale=<0.3 0.3 0.3>
    physicsBodyDefinitions [{
      sphereCollisionRadius=0
      boundBone {boneName=dyn_candy13}
      boxExtents=<12 8 35>
      constraintSetup {angularLimitsMax=<20 10 5>, angularLimitsMin=<-20 -10 -5>, angularTarget=<1 0 0>}
      localJointOffset=<0 0 -6>
     }]
   }
   n14 = Anim Dynamics - Bone: dyn_candy11_b!(ComponentPose=@n13, ExternalForce="0,0,0") -> (Pose) {
    alpha=0.35
    angularDampingOverride=0.998
    angularSpringConstant=100
    bAngularSpring=true
    bGravityOverrideInSimSpace=true
    bOverrideAngularDamping=true
    gravityScale=0.5
    simulationSpace=RootRelative
    alphaBoolBlend {blendOption=Linear}
    boundBone {boneName=dyn_candy11_b}
    componentAppliedLinearAccClamp=<3000 3000 3000>
    componentLinearAccScale=<0.4 0.4 0.5>
    componentLinearVelScale=<0.3 0.3 0.3>
    physicsBodyDefinitions [{
      sphereCollisionRadius=0
      boundBone {boneName=dyn_candy11_b}
      boxExtents=<12 8 35>
      constraintSetup {angularLimitsMax=<10 5 5>, angularLimitsMin=<-10 -5 -5>, angularTarget=<1 0 0>}
      localJointOffset=<0 0 -6>
     }]
   }
   n15 = Get FE Alpha() -> (FE Alpha:double)
   n16 = Anim Dynamics - Bone: dyn_candy1!(ComponentPose=Local To Component(LocalPose=Use cached pose 'Sim'()), Alpha:float=@n15, ExternalForce="0,0,0") -> (Pose) {
    alpha=0.5
    angularDampingOverride=0.998
    angularSpringConstant=100
    bAngularSpring=true
    bGravityOverrideInSimSpace=true
    bOverrideAngularDamping=true
    gravityScale=0.5
    simulationSpace=RootRelative
    alphaBoolBlend {blendOption=Linear}
    boundBone {boneName=dyn_candy1}
    componentAppliedLinearAccClamp=<3000 3000 3000>
    componentLinearAccScale=<0.3 0.3 0.4>
    componentLinearVelScale=<0.2 0.2 0.2>
    physicsBodyDefinitions [{
      sphereCollisionRadius=0
      boundBone {boneName=dyn_candy1}
      boxExtents=<8 30 30>
      constraintSetup {angularLimitsMax=<5 5 5>, angularLimitsMin=<-5 -5 -5>, angularTarget=<1 0 0>}
      localJointOffset=<0 0 -6>
     }]
   }
   n17 = Anim Dynamics - Bone: dyn_candy2!(ComponentPose=@n16, Alpha:float=@n15, ExternalForce="0,0,0") -> (Pose) {
    alpha=0.5
    angularDampingOverride=0.998
    angularSpringConstant=75
    bAngularSpring=true
    bGravityOverrideInSimSpace=true
    bLinearSpring=true
    bOverrideAngularDamping=true
    gravityScale=0.5
    linearDampingOverride=0.998
    linearSpringConstant=10
    simulationSpace=RootRelative
    alphaBoolBlend {blendOption=Linear}
    boundBone {boneName=dyn_candy2}
    componentAppliedLinearAccClamp=<3000 3000 3000>
    componentLinearAccScale=<0.4 0.4 0.5>
    componentLinearVelScale=<0.3 0.3 0.3>
    physicsBodyDefinitions [{
      sphereCollisionRadius=0
      boundBone {boneName=dyn_candy2}
      boxExtents=<12 8 30>
      constraintSetup {angularLimitsMax=<20 10 5>, angularLimitsMin=<-20 -10 -5>, angularTarget=<1 0 0>, linearAxesMax=<0 0 2>}
      localJointOffset=<0 0 -6>
     }]
   }
   n18 = Anim Dynamics - Bone: dyn_candy3!(ComponentPose=@n17, Alpha:float=@n15, ExternalForce="0,0,0") -> (Pose) {
    alpha=0.5
    angularDampingOverride=0.998
    angularSpringConstant=100
    bAngularSpring=true
    bGravityOverrideInSimSpace=true
    bLinearSpring=true
    bOverrideAngularDamping=true
    gravityScale=0.5
    linearDampingOverride=0.998
    linearSpringConstant=5
    simulationSpace=RootRelative
    alphaBoolBlend {blendOption=Linear}
    boundBone {boneName=dyn_candy3}
    componentAppliedLinearAccClamp=<3000 3000 3000>
    componentLinearAccScale=<0.4 0.4 0.5>
    componentLinearVelScale=<0.3 0.3 0.3>
    physicsBodyDefinitions [{
      sphereCollisionRadius=0
      boundBone {boneName=dyn_candy3}
      boxExtents=<12 8 25>
      constraintSetup {angularLimitsMax=<20 10 5>, angularLimitsMin=<-20 -10 -5>, angularTarget=<1 0 0>, linearAxesMax=<0 0 1>}
      localJointOffset=<0 0 -6>
     }]
   }
   n19 = Anim Dynamics - Bone: dyn_candy4!(ComponentPose=@n18, Alpha:float=@n15, ExternalForce="0,0,0") -> (Pose) {
    alpha=0.5
    angularDampingOverride=0.998
    angularSpringConstant=100
    bAngularSpring=true
    bGravityOverrideInSimSpace=true
    bLinearSpring=true
    bOverrideAngularDamping=true
    gravityScale=0.5
    linearDampingOverride=0.998
    linearSpringConstant=5
    simulationSpace=RootRelative
    alphaBoolBlend {blendOption=Linear}
    boundBone {boneName=dyn_candy4}
    componentAppliedLinearAccClamp=<3000 3000 3000>
    componentLinearAccScale=<0.4 0.4 0.5>
    componentLinearVelScale=<0.3 0.3 0.3>
    physicsBodyDefinitions [{
      sphereCollisionRadius=0
      boundBone {boneName=dyn_candy4}
      boxExtents=<12 8 25>
      constraintSetup {angularLimitsMax=<20 10 5>, angularLimitsMin=<-20 -10 -5>, angularTarget=<1 0 0>, linearAxesMax=<0 0 0.5>}
      localJointOffset=<0 0 -6>
     }]
   }
   n20 = Anim Dynamics - Bone: dyn_candy5!(ComponentPose=@n19, Alpha:float=@n15, ExternalForce="0,0,0") -> (Pose) {
    alpha=0.5
    angularDampingOverride=0.998
    angularSpringConstant=100
    bAngularSpring=true
    bGravityOverrideInSimSpace=true
    bLinearSpring=true
    bOverrideAngularDamping=true
    gravityScale=0.5
    linearDampingOverride=0.998
    linearSpringConstant=15
    simulationSpace=RootRelative
    alphaBoolBlend {blendOption=Linear}
    boundBone {boneName=dyn_candy5}
    componentAppliedLinearAccClamp=<3000 3000 3000>
    componentLinearAccScale=<0.4 0.4 0.5>
    componentLinearVelScale=<0.3 0.3 0.3>
    physicsBodyDefinitions [{
      sphereCollisionRadius=0
      boundBone {boneName=dyn_candy5}
      boxExtents=<12 8 35>
      constraintSetup {angularLimitsMax=<20 10 5>, angularLimitsMin=<-20 -10 -5>, angularTarget=<1 0 0>, linearAxesMax=<0 0 1>}
      localJointOffset=<0 0 -6>
     }]
   }
   n21 = Anim Dynamics - Bone: dyn_candy6!(ComponentPose=@n20, Alpha:float=@n15, ExternalForce="0,0,0") -> (Pose) {
    alpha=0.5
    angularDampingOverride=0.998
    angularSpringConstant=100
    bAngularSpring=true
    bGravityOverrideInSimSpace=true
    bLinearSpring=true
    bOverrideAngularDamping=true
    gravityScale=0.5
    linearDampingOverride=0.998
    linearSpringConstant=5
    simulationSpace=RootRelative
    alphaBoolBlend {blendOption=Linear}
    boundBone {boneName=dyn_candy6}
    componentAppliedLinearAccClamp=<3000 3000 3000>
    componentLinearAccScale=<0.4 0.4 0.5>
    componentLinearVelScale=<0.3 0.3 0.3>
    physicsBodyDefinitions [{
      sphereCollisionRadius=0
      boundBone {boneName=dyn_candy6}
      boxExtents=<12 8 25>
      constraintSetup {angularLimitsMax=<20 10 5>, angularLimitsMin=<-20 -10 -5>, angularTarget=<1 0 0>, linearAxesMax=<0 0 1>}
      localJointOffset=<0 0 -6>
     }]
   }
   n22 = Anim Dynamics - Bone: dyn_candy8!(ComponentPose=@n21, Alpha:float=@n15, ExternalForce="0,0,0") -> (Pose) {
    alpha=0.5
    angularDampingOverride=0.998
    angularSpringConstant=75
    bAngularSpring=true
    bGravityOverrideInSimSpace=true
    bLinearSpring=true
    bOverrideAngularDamping=true
    gravityScale=0.5
    linearDampingOverride=0.998
    linearSpringConstant=5
    simulationSpace=RootRelative
    alphaBoolBlend {blendOption=Linear}
    boundBone {boneName=dyn_candy8}
    componentAppliedLinearAccClamp=<3000 3000 3000>
    componentLinearAccScale=<0.4 0.4 0.5>
    componentLinearVelScale=<0.3 0.3 0.3>
    physicsBodyDefinitions [{
      sphereCollisionRadius=0
      boundBone {boneName=dyn_candy8}
      boxExtents=<12 8 30>
      constraintSetup {angularLimitsMax=<20 10 5>, angularLimitsMin=<-20 -10 -5>, angularTarget=<1 0 0>, linearAxesMax=<0 0 1>}
      localJointOffset=<0 0 -6>
     }]
   }
   n23 = Anim Dynamics - Bone: dyn_candy7!(ComponentPose=@n22, Alpha:float=@n15, ExternalForce="0,0,0") -> (Pose) {
    alpha=0.5
    angularDampingOverride=0.998
    angularSpringConstant=100
    bAngularSpring=true
    bGravityOverrideInSimSpace=true
    bLinearSpring=true
    bOverrideAngularDamping=true
    gravityScale=0.5
    linearDampingOverride=0.998
    linearSpringConstant=5
    simulationSpace=RootRelative
    alphaBoolBlend {blendOption=Linear}
    boundBone {boneName=dyn_candy7}
    componentAppliedLinearAccClamp=<3000 3000 3000>
    componentLinearAccScale=<0.4 0.4 0.5>
    componentLinearVelScale=<0.3 0.3 0.3>
    physicsBodyDefinitions [{
      sphereCollisionRadius=0
      boundBone {boneName=dyn_candy7}
      boxExtents=<12 8 35>
      constraintSetup {angularLimitsMax=<20 10 5>, angularLimitsMin=<-20 -10 -5>, angularTarget=<1 0 0>, linearAxesMax=<0 0 2>}
      localJointOffset=<0 0 -6>
     }]
   }
   n24 = Anim Dynamics - Bone: dyn_candy9!(ComponentPose=@n23, Alpha:float=@n15, ExternalForce="0,0,0") -> (Pose) {
    alpha=0.5
    angularDampingOverride=0.998
    angularSpringConstant=125
    bAngularSpring=true
    bGravityOverrideInSimSpace=true
    bLinearSpring=true
    bOverrideAngularDamping=true
    gravityScale=0.5
    linearDampingOverride=0.998
    linearSpringConstant=15
    simulationSpace=RootRelative
    alphaBoolBlend {blendOption=Linear}
    boundBone {boneName=dyn_candy9}
    componentAppliedLinearAccClamp=<3000 3000 3000>
    componentLinearAccScale=<0.4 0.4 0.5>
    componentLinearVelScale=<0.3 0.3 0.3>
    physicsBodyDefinitions [{
      sphereCollisionRadius=0
      boundBone {boneName=dyn_candy9}
      boxExtents=<12 8 25>
      constraintSetup {angularLimitsMax=<20 10 5>, angularLimitsMin=<-20 -10 -5>, angularTarget=<1 0 0>, linearAxesMax=<0 0 4>}
      localJointOffset=<0 0 -6>
     }]
   }
   n25 = Anim Dynamics - Bone: dyn_candy10!(ComponentPose=@n24, Alpha:float=@n15, ExternalForce="0,0,0") -> (Pose) {
    alpha=0.5
    angularDampingOverride=0.998
    angularSpringConstant=100
    bAngularSpring=true
    bGravityOverrideInSimSpace=true
    bLinearSpring=true
    bOverrideAngularDamping=true
    gravityScale=0.5
    linearDampingOverride=0.998
    linearSpringConstant=20
    simulationSpace=RootRelative
    alphaBoolBlend {blendOption=Linear}
    boundBone {boneName=dyn_candy10}
    componentAppliedLinearAccClamp=<3000 3000 3000>
    componentLinearAccScale=<0.4 0.4 0.5>
    componentLinearVelScale=<0.3 0.3 0.3>
    physicsBodyDefinitions [{
      sphereCollisionRadius=0
      boundBone {boneName=dyn_candy10}
      boxExtents=<12 8 20>
      constraintSetup {angularLimitsMax=<20 10 5>, angularLimitsMin=<-20 -10 -5>, angularTarget=<1 0 0>}
      localJointOffset=<0 0 -6>
     }]
   }
   n26 = Anim Dynamics - Bone: dyn_candy11!(ComponentPose=@n25, Alpha:float=@n15, ExternalForce="0,0,0") -> (Pose) {
    alpha=0.5
    angularDampingOverride=0.998
    angularSpringConstant=75
    bAngularSpring=true
    bGravityOverrideInSimSpace=true
    bLinearSpring=true
    bOverrideAngularDamping=true
    gravityScale=0.5
    linearDampingOverride=0.998
    linearSpringConstant=15
    simulationSpace=RootRelative
    alphaBoolBlend {blendOption=Linear}
    boundBone {boneName=dyn_candy11}
    componentAppliedLinearAccClamp=<3000 3000 3000>
    componentLinearAccScale=<0.4 0.4 0.5>
    componentLinearVelScale=<0.3 0.3 0.3>
    physicsBodyDefinitions [{
      sphereCollisionRadius=0
      boundBone {boneName=dyn_candy11}
      boxExtents=<12 8 30>
      constraintSetup {angularLimitsMax=<25 15 10>, angularLimitsMin=<-25 -15 -10>, angularTarget=<1 0 0>, linearAxesMax=<0 0 2.5>}
      localJointOffset=<0 0 -6>
     }]
   }
   n27 = Anim Dynamics - Bone: dyn_candy12!(ComponentPose=@n26, Alpha:float=@n15, ExternalForce="0,0,0") -> (Pose) {
    alpha=0.5
    angularDampingOverride=0.998
    angularSpringConstant=100
    bAngularSpring=true
    bGravityOverrideInSimSpace=true
    bLinearSpring=true
    bOverrideAngularDamping=true
    gravityScale=0.5
    linearDampingOverride=0.998
    linearSpringConstant=5
    simulationSpace=RootRelative
    alphaBoolBlend {blendOption=Linear}
    boundBone {boneName=dyn_candy12}
    componentAppliedLinearAccClamp=<3000 3000 3000>
    componentLinearAccScale=<0.4 0.4 0.5>
    componentLinearVelScale=<0.3 0.3 0.3>
    physicsBodyDefinitions [{
      sphereCollisionRadius=0
      boundBone {boneName=dyn_candy12}
      boxExtents=<12 8 30>
      constraintSetup {angularLimitsMax=<20 10 5>, angularLimitsMin=<-20 -10 -5>, angularTarget=<1 0 0>}
      localJointOffset=<0 0 -6>
     }]
   }
   n28 = Anim Dynamics - Bone: dyn_candy13!(ComponentPose=@n27, Alpha:float=@n15, ExternalForce="0,0,0") -> (Pose) {
    alpha=0.5
    angularDampingOverride=0.998
    angularSpringConstant=100
    bAngularSpring=true
    bGravityOverrideInSimSpace=true
    bLinearSpring=true
    bOverrideAngularDamping=true
    gravityScale=0.5
    linearDampingOverride=0.998
    linearSpringConstant=10
    simulationSpace=RootRelative
    alphaBoolBlend {blendOption=Linear}
    boundBone {boneName=dyn_candy13}
    componentAppliedLinearAccClamp=<3000 3000 3000>
    componentLinearAccScale=<0.4 0.4 0.5>
    componentLinearVelScale=<0.3 0.3 0.3>
    physicsBodyDefinitions [{
      sphereCollisionRadius=0
      boundBone {boneName=dyn_candy13}
      boxExtents=<12 8 35>
      constraintSetup {angularLimitsMax=<20 10 5>, angularLimitsMin=<-20 -10 -5>, angularTarget=<1 0 0>, linearAxesMax=<0 0 2>}
      localJointOffset=<0 0 -6>
     }]
   }
   n29 = Anim Dynamics - Bone: dyn_candy11_b!(ComponentPose=@n28, Alpha:float=@n15, ExternalForce="0,0,0") -> (Pose) {
    alpha=0.5
    angularDampingOverride=0.998
    angularSpringConstant=100
    bAngularSpring=true
    bGravityOverrideInSimSpace=true
    bLinearSpring=true
    bOverrideAngularDamping=true
    gravityScale=0.5
    linearDampingOverride=0.998
    linearSpringConstant=25
    simulationSpace=RootRelative
    alphaBoolBlend {blendOption=Linear}
    boundBone {boneName=dyn_candy11_b}
    componentAppliedLinearAccClamp=<3000 3000 3000>
    componentLinearAccScale=<0.4 0.4 0.5>
    componentLinearVelScale=<0.3 0.3 0.3>
    physicsBodyDefinitions [{
      sphereCollisionRadius=0
      boundBone {boneName=dyn_candy11_b}
      boxExtents=<12 8 35>
      constraintSetup {angularLimitsMax=<10 5 5>, angularLimitsMin=<-10 -5 -5>, angularTarget=<1 0 0>, linearAxesMax=<0 0 4>}
      localJointOffset=<0 0 -6>
     }]
   }
   Output Pose(Result=Blend Poses by bool(BlendPose_0=Component To Local(ComponentPose=@n14), BlendPose_1=Component To Local(ComponentPose=@n29), bActiveValue=Get bIsSkyDiving()) {blendType=Linear}) {layerGroup=None}
   n30 = RigidBody!(ComponentPose=Local To Component(LocalPose=Local Space Ref Pose()), Alpha, OverrideWorldGravity:Vector="0,0,-980", ExternalForce) -> (Pose) {
    bClampLinearTranslationLimitToRefPose=true
    bOverrideWorldGravity=true
    overridePhysicsAsset=&/M_MED_Disguise_Pack_Physics
    simulationSpace=BaseBoneSpace
    alphaBoolBlend {blendOption=Linear}
    baseBoneRef {boneName=root}
    componentAppliedLinearAccClamp=<3000 3000 3000>
    componentLinearAccScale=<0.3 0.3 0.5>
    componentLinearVelScale=<0.2 0.2 0.2>
    overrideWorldGravity {z=-980}
   }
   n31 = "Transform (Modify) Bone - Bone: dyn_candy1"(ComponentPose=@n30, Alpha=Get Flap Alpha() -> (Flap Alpha:double), Translation="0,0,0", Rotation="0,0,45", Scale="1,1,1") {
    rotationMode=BMM_Additive
    rotationSpace=BCS_BoneSpace
    alphaBoolBlend {blendOption=Linear}
    boneToModify {boneName=dyn_candy1}
    rotation {roll=45}
   }
   n32 = "Transform (Modify) Bone - Bone: dyn_candyGroup_MSTR"(ComponentPose=@n31, Alpha=Get Falling Alpha() -> (Falling Alpha:double), Translation="0,0,1.5", Rotation="0,0,0", Scale="1,1,1") {
    rotationMode=BMM_Additive
    rotationSpace=BCS_BoneSpace
    translationMode=BMM_Additive
    translationSpace=BCS_BoneSpace
    alphaBoolBlend {blendInTime=0.4, blendOption=Linear, blendOutTime=0.2}
    alphaScaleBiasClamp {bInterpResult=true, interpSpeedDecreasing=100, interpSpeedIncreasing=2.5}
    boneToModify {boneName=dyn_candyGroup_MSTR}
    translation {z=1.5}
   }
   n33 = "Transform (Modify) Bone - Bone: dyn_lollipop_MSTR"(ComponentPose=@n32, Alpha=Get Pickaxe Alpha() -> (Pickaxe Alpha:double), Translation="0,0,0", Rotation="5,-22.5,10", Scale="1,1,1") {
    rotationMode=BMM_Additive
    rotationSpace=BCS_BoneSpace
    translationMode=BMM_Additive
    translationSpace=BCS_BoneSpace
    alphaBoolBlend {blendInTime=0.4, blendOption=Linear, blendOutTime=0.2}
    alphaScaleBiasClamp {bInterpResult=true, interpSpeedDecreasing=100, interpSpeedIncreasing=3}
    boneToModify {boneName=dyn_lollipop_MSTR}
    rotation=<rot 5 -22.5 10>
   }
   n34 = Get bIsInFrontEnd()
   n35 = "Constraint - dyn_lollipop (P)"!(ComponentPose=@n33, bAlphaBoolEnabled:bool=@n34, ConstraintWeights_0=1) -> (Pose) {
    alphaInputType=Bool
    alphaBoolBlend {blendOption=Linear}
    boneToModify {boneName=dyn_lollipop}
    constraintSetup [{targetBone {boneName=dyn_lollipop_MSTR}}]
    constraintWeights [1]
   }
   n36 = "Transform (Modify) Bone - Bone: dyn_lollipop"!(ComponentPose=@n35, bAlphaBoolEnabled:bool=@n34, Translation="0,0,0", Rotation="5,-20,15", Scale="1,1,1") -> (Pose) {
    alphaInputType=Bool
    rotationMode=BMM_Additive
    rotationSpace=BCS_BoneSpace
    alphaBoolBlend {blendOption=Linear}
    boneToModify {boneName=dyn_lollipop}
    rotation=<rot 5 -20 15>
   }
   Save cached pose 'Sim'(Pose=Component To Local(ComponentPose=@n36)) {cacheName=Sim}
  ]}
 EventGraph {nodes [
   Event BlueprintUpdateAnimation()
   > n1 = Sequence!(execute) -> (then_0, then_1, then_2:Exec, then_3:Exec, then_4:Exec)
   n2 = BreakTransform(InTransform=GetDeltaTransformFromRefPose(self=GetPartSkeletalMeshComponent(), BoneName=dyn_flap, BaseName=dyn_flap_MSTR))
   > Set Flap Alpha(Flap Alpha:double=MapRangeClamped(Value=BreakRotator(InRot=@n2.Rotation).Yaw, InRangeA=20, InRangeB=70)) -> (Output_Get:double)
   n3 = OR Boolean(A=Get bIsSkydiveDiveMode(), B=Get bIsParachuteOpened())
   n4 = Branch(execute=@n1.then_1, Condition=AND Boolean(A=Get bIsFalling(), B=NOT Boolean(A=@n3)))
   > Set Falling Alpha(Falling Alpha:double=SelectFloat(A=0.5, B=1, bPickA=Get bIsSkyDiving())) -> (Output_Get:double)
   Set Falling Alpha(execute=@n4.else, Falling Alpha=0)
   Set Pickaxe Alpha(execute=@n1.then_2, Pickaxe Alpha:double=0) -> (Output_Get:double)
   > n5 = Switch on EFortWeaponCoreAnimation!(
    execute
    Selection:EFortWeaponCoreAnimation Enum=GetWeaponCoreAnim!(self:Fort Player Animation Interface Object Reference=GetOwnerAnimationInterface!(self:Custom Character Part Anim Instance Object Reference) -> (ReturnValue:Fort Player Animation Interface Object Reference)) -> (ReturnValue:EFortWeaponCoreAnimation Enum)
   ) -> (
    Melee:Exec
    Pistol:Exec
    Shotgun:Exec
    PaperBlueprint:Exec
    Rifle:Exec
    MeleeOneHand:Exec
    MachinePistol:Exec
    RocketLauncher:Exec
    GrenadeLauncher:Exec
    GoingCommando:Exec
    AssaultRifle:Exec
    TacticalShotgun:Exec
    SniperRifle:Exec
    TrapPlacement:Exec
    ShoulderLauncher:Exec
    AbilityDecoTool:Exec
    Crossbow:Exec
    C4:Exec
    RemoteControl:Exec
    DualWield:Exec
    AR_BullPup:Exec
    AR_ForwardGrip:Exec
    MedPackPaddles:Exec
    SMG_P90:Exec
    AR_DrumGun:Exec
    Consumable_Small:Exec
    Consumable_Large:Exec
    Balloon:Exec
    MountedTurret:Exec
    CreativeTool:Exec
    ExplosiveBow:Exec
    AshtonIndigo:Exec
    AshtonChicago:Exec
    MeleeDualWield:Exec
    Unarmed:Exec
   )
   Set Pickaxe Alpha(execute=[@n5.Melee, @n5.MeleeOneHand], Pickaxe Alpha=1)
   n6 = Branch(execute=@n1.then_4, Condition=Get bIsInFrontEnd())
   > Set FE Alpha(FE Alpha:double=0.25) -> (Output_Get:double)
   Set FE Alpha(execute=@n6.else, FE Alpha=0.5)
   TryGetPawnOwner()
  ]}
]
variables [
 Falling Alpha {type=number}
 FE Alpha {type=number}
 Flap Alpha {type=number}
 Gravity Override {type=Vector}
 isFemale {type=boolean}
 Pickaxe Alpha {type=number}
]
